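# Material function for the WiryPerk backpack's animated water-cooling effect, outputting slurp base color, FX mask, roughness, specular and emissive.
# Panning, rotating swirl noise with a sine-driven offset samples a curve atlas for color.
# Two panning BumpOffset bubble layers, modulated by scrolling noise and Fresnel, supply fake depth.
# Emissive sums bubbles with AO- and Fresnel-shaded color, scales by time of day and hue-shifts over time, while base color fades toward black.
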MaterialFunction /BRCosmetics/Accessories/FORT_Backpacks/Backpack_F_MED_WiryPerk/Materials/MF_WiryPerk_WaterCooling
graph {nodes [
  Constant2Vector()
  Constant() {r=0.2}
  FunctionInput() {bUsePreviewValueAsDefault=true, inputName=baseColor}
  FunctionOutput(None=LinearInterpolate(A=NamedRerouteUsage(), Alpha=ScalarParameter() {defaultValue=1, group=Slurp, parameterName=Boots_emissiveWeight}) {constB=0}) {outputName=Slurp_bascolor}
  n1 = ScalarParameter() {defaultValue=1, group=Slurp, parameterName=SimpleSlurpWeight}
  FunctionOutput(None=Multiply(A=FunctionInput() {bUsePreviewValueAsDefault=true, inputName=input_fxmask, inputType=FunctionInput_Scalar}, B=LinearInterpolate(Alpha=@n1))) {outputName=Slurp_FX mask}
  n2 = ScalarParameter() {defaultValue=0.2, group=Slurp, parameterName=Slurp_Roughness}
  n3 = FunctionInput() {bUsePreviewValueAsDefault=true, inputName=S_map}
  n4 = ComponentMask(None=@n3) {b=true}
  n5 = ScalarParameter() {defaultValue=0.5, group=Slurp, parameterName=Slurp_RoughnessMap_Intensity}
  FunctionOutput(None=NamedRerouteDeclaration(None=LinearInterpolate(A=@n2, B=@n4, Alpha=@n5)) {name=Roughness, nodeColor=<rgba 0 1 0.7188 1>}) {outputName=SlurpRougness}
  n6 = ScalarParameter() {defaultValue=0.5, group=Slurp, parameterName=Slurp_Specular}
  n7 = ScalarParameter() {defaultValue=1, group=Slurp, parameterName=Slurp_SpecularMap_Intensity}
  n8 = ComponentMask(None=@n3) {r=true}
  FunctionOutput(None=NamedRerouteDeclaration(None=LinearInterpolate(A=@n6, B=@n8, Alpha=@n7)) {name=SPC, nodeColor=<rgba 1 0 0.7266 1>}) {outputName=Slurp_Spc}
  n9 = Multiply(A=Time()) {constB=0.06}
  n10 = ScalarParameter() {defaultValue=0.66, group=Slurp, parameterName=Slurp_TimeofDayScale_Morning}
  n11 = ScalarParameter() {defaultValue=1, group=Slurp, parameterName=Slurp_TimeofDayScale_Day}
  n12 = ScalarParameter() {defaultValue=0.66, group=Slurp, parameterName=Slurp_TimeofDayScale_Evening}
  n13 = ScalarParameter() {defaultValue=0.5, group=Slurp, parameterName=Slurp_TimeofDayScale_Night}
  n14 = Multiply(A=NamedRerouteUsage(), B=ScalarParameter() {defaultValue=0.45, group=Water_Cooling, parameterName=BubbleBrightness}) {constB=0.75}
  n15 = Power(
   Base=ComponentMask(None=FunctionInput() {bUsePreviewValueAsDefault=true, inputName=AO}) {r=true}
   Exp=ScalarParameter() {defaultValue=1.3, group=Water_Cooling, parameterName=Boots_AOintensity}
  )
  n16 = OneMinus(None=Fresnel(ExponentIn=Constant() {r=1}, BaseReflectFractionIn=Constant() {r=-0.0837}))
  n17 = LinearInterpolate(
   A=ScalarParameter() {defaultValue=0.8, group=Water_Cooling, parameterName=FresnelBrightnessMax}
   B=ScalarParameter() {defaultValue=0.6, group=Water_Cooling, parameterName=FresnelBrightnessMin}
   Alpha=Saturate(None=Power(Base=@n16))
  ) {constA=0.5, constB=0.9}
  n18 = LinearInterpolate(
   A=Multiply(A=Multiply(A=@n15, B=@n17), B=NamedRerouteUsage())
   Alpha=ScalarParameter() {defaultValue=0.3, group=Water_Cooling, parameterName=WaterCooling_Global_brightness}
  ) {constB=0}
  n19 = LinearInterpolate(
   A=Constant()
   B=Multiply(
    A=MaterialFunctionCall(Value_Morning=@n10, Value_Day=@n11, Value_Evening=@n12, Value_Night=@n13) {materialFunction=&Scalar_Time_of_Day}.Result
    B=Add(A=@n14, B=@n18)
   )
   Alpha=ScalarParameter() {defaultValue=1, group=Slurp, parameterName=Boots_emissiveWeight}
  )
  FunctionOutput(None=NamedRerouteDeclaration(None=MaterialFunctionCall(Hue Shift Percentage=@n9, Texture=@n19) {materialFunction=&HueShift}.Result) {name=Emissive, nodeColor=<rgba 0 1 0.1797 1>}) {outputName=Slurp_emiss}
  Max()
  n20 = Constant2Vector()
  n21 = AppendVector(A=Constant() {r=-0.067}, B=Constant())
  n22 = ComponentMask(None=NamedRerouteUsage()) {r=true}
  n23 = Multiply(A=Add(A=TextureCoordinate(), B=LinearInterpolate(A=@n20, B=@n21, Alpha=@n22))) {constB=2}
  n24 = ScalarParameter() {defaultValue=0.16, parameterName=waterSpeed}
  n25 = LinearInterpolate(A=AppendVector(A=Constant(), B=@n24), B=AppendVector(A=@n24, B=Constant()), Alpha=ComponentMask(None=NamedRerouteUsage()) {r=true})
  n26 = Constant()
  n27 = Constant() {r=0.25}
  n28 = ComponentMask(None=NamedRerouteUsage()) {r=true}
  n29 = TextureSample(
   UVs=MaterialFunctionCall(UVs=Panner(Coordinate=@n23, Speed=@n25), "Rotation Angle (0-1)"=LinearInterpolate(A=@n26, B=@n27, Alpha=@n28) {constA=0.5, constB=0}) {materialFunction=&CustomRotator}.Rotated Values
   Tex=TextureObject() {samplerType=SAMPLERTYPE_Masks, texture=/Game/Global/Textures/Default/Masks_and_Gradients/T_VertGradient_MASK}
  ) {
   samplerSource=SSM_Wrap_WorldGroupSettings
   samplerType=SAMPLERTYPE_Masks
   texture=/Game/Weapons/WeaponSkins/Textures/T_Swirly_Noise_highContrast
  }
  n30 = ScalarParameter() {defaultValue=0.5, group=WiryPerk, parameterName=ColorOffset}
  n31 = Multiply(A=Time()) {constB=0.1}
  n32 = ScalarParameter() {defaultValue=0.5, group=WiryPerk, parameterName=SineSpeed, sortPriority=2}
  n33 = ConstantBiasScale(None=Sine(None=Multiply(A=@n31, B=@n32)))
  n34 = Multiply(A=Add(A=@n29.R, B=LinearInterpolate(B=@n30, Alpha=@n33) {constB=0.5}))
  n35 = CurveAtlasRowParameter(CurveTime=ComponentMask(None=@n34) {g=true}) {
   atlas=/BRCosmetics/Gadgets/Assets/VinderTech_GliderChute/Glider_WiryPerk/Materials/WiryPerk_ColorAtlas
   curve=/BRCosmetics/Gadgets/Assets/VinderTech_GliderChute/Glider_WiryPerk/Materials/WiryPerk_CurveBackPack
   defaultValue=1
   group=06_Advanced Emissive
   parameterName=PanningEmissiveColorCurve
   sortPriority=6
  }
  NamedRerouteDeclaration(None=Multiply(A=@n35) {constB=0.99}) {name=InputBaseColor, nodeColor=<rgba 0.7344 1 0 1>}
  NamedRerouteDeclaration(None=FunctionInput() {bUsePreviewValueAsDefault=true, inputName=WaterCooling Mask}) {name=WaterCooling Mask, nodeColor=<rgba 1 0 0.3516 1>}
  n36 = ScalarParameter() {defaultValue=0.2, group=Water_Cooling, parameterName=Bubble_ViewDistortion_SpeedX}
  n37 = ScalarParameter() {group=Water_Cooling, parameterName=Bubble_ViewDistortion_SpeedY}
  n38 = TextureSample(UVs=Panner(Coordinate=TextureCoordinate(), Time=Time(), Speed=AppendVector(A=@n36, B=@n37))) {
   constMipValue=0
   texture=/BRCosmetics/Characters/Player/Female/Medium/Bodies/F_MED_Twin_Dark/Skins/Twin_Light_Purple/Textures/T_twinlightPurpleBlurredNoise
  }
  n39 = Power(Base=@n38.R)
  n40 = BumpOffset(Coordinate=Multiply(A=TextureCoordinate(), B=Constant() {r=20}), Height=Constant() {r=-0.7})
  n41 = LinearInterpolate(A=Constant2Vector() {g=0.2}, B=Constant2Vector() {r=0.2}, Alpha=ComponentMask(None=NamedRerouteUsage()) {r=true})
  n42 = Multiply(A=@n41, B=ScalarParameter() {defaultValue=3, group=Water_Cooling, parameterName=Bubble_Mul}) {constB=0.5}
  n43 = BumpOffset(Coordinate=Multiply(A=TextureCoordinate(), B=Constant() {r=20}), Height=Constant() {r=-0.1})
  n44 = Max(
   A=TextureSample(UVs=Panner(Coordinate=@n40, Speed=@n42) {speedY=0.4}) {texture=/BRCosmetics/Characters/Player/Female/Medium/Bodies/F_MED_Twin_Dark/Skins/Twin_Light_Purple/Textures/T_twinlightBubble}.R
   B=TextureSample(UVs=Panner(Coordinate=@n43, Speed=@n41) {speedY=0.1}) {texture=&T_twinlightBubble}.G
  )
  n45 = Constant() {r=0.32}
  n46 = Multiply(A=Desaturation(None=Multiply(A=@n44, B=@n45), Fraction=Constant() {r=1}), B=NamedRerouteUsage())
  n47 = OneMinus(None=Fresnel(ExponentIn=Constant() {r=1}, BaseReflectFractionIn=Constant() {r=-1}))
  NamedRerouteDeclaration(None=LinearInterpolate(B=Multiply(A=@n39, B=@n46), Alpha=Power(Base=@n47))) {name=simple bubble, nodeColor=<rgba 0 0.3906 1 1>}
  TextureSampleParameter2D(UVs=ComponentMask(None=@n34) {g=true, r=true}) {
   group=06_Advanced Emissive
   parameterName=PanningEmissiveColorGradient
   samplerSource=SSM_Wrap_WorldGroupSettings
   sortPriority=6
   texture=/Game/Global/Textures/Default/Masks_and_Gradients/3pxRGBImage
  }
 ]}
