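# Material function for the WiryPerk backpack cosmetic, overriding base color, emissive, anisotropy and tangent on incoming material attributes to animate its computer-like parts.
# Custom rotators spin the disc UVs and a mask driving a three-color IC chip light.
# Time-based hue shift and panning noise animate the emissive, while a nested WaterCooling function supplies coolant color, glow and roughness.
# Texture masks and UV sphere masks isolate parts, and TextAnimation scalar parameters provide animation control.
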MaterialFunction /BRCosmetics/Accessories/FORT_Backpacks/Backpack_F_MED_WiryPerk/Materials/MF_WiryPerkBackpack
graph {nodes [
  n1 = NamedRerouteUsage()
  n2 = NamedRerouteUsage()
  n3 = NamedRerouteUsage()
  n4 = NamedRerouteUsage()
  n5 = NamedRerouteUsage()
  n6 = NamedRerouteUsage()
  n7 = SetMaterialAttributes!(
   MaterialAttributes=GetMaterialAttributes(None=FunctionInput() {inputType=FunctionInput_MaterialAttributes}) {attributeGetTypes [#1, #2]}.MaterialAttributes
   BaseColor=LinearInterpolate(A=LinearInterpolate(A=@n1, B=@n2, Alpha=@n3), Alpha=NamedRerouteUsage()) {constB=0}
   EmissiveColor=LinearInterpolate(A=LinearInterpolate(A=@n4, B=@n5, Alpha=@n6), B=NamedRerouteUsage(), Alpha=NamedRerouteUsage())
   Anisotropy=LinearInterpolate(B=ScalarParameter() {defaultValue=0.38, parameterName=anisotropy}, Alpha=NamedRerouteUsage()) {constB=0.5}
   Tangent=LinearInterpolate(A=Constant3Vector() {constant=<rgba 1 0 0 0>}, B=Constant3Vector() {constant=<rgba 0 1 0 1>}, Alpha=NamedRerouteUsage())
  ) {attributeSetTypes [
    #1
    #2
    #3
    #4
   ]}
  FunctionOutput(None=@n7)
  n8 = ScalarParameter() {defaultValue=0.7, parameterName=DiscBrightness}
  n9 = NamedRerouteUsage()
  n10 = Saturate(None=Multiply(A=Multiply(A=@n9) {constB=1.5}, B=NamedRerouteUsage()))
  NamedRerouteDeclaration(None=LinearInterpolate(B=LinearInterpolate(A=@n8, Alpha=@n10) {constA=1, constB=0.6}, Alpha=NamedRerouteUsage()) {constA=1}) {name=DiscBrightness, nodeColor=<rgba 0 0.3906 1 1>}
  n11 = TextureSample() {samplerType=SAMPLERTYPE_Masks, texture=/BRCosmetics/Accessories/FORT_Backpacks/Backpack_F_MED_WiryPerk/Textures/T_WiryPerk_Backpack_TA_mask}
  NamedRerouteDeclaration(None=@n11.R) {name=Disc_mask, nodeColor=<rgba 0.2656 0 1 1>}
  n12 = TextureCoordinate()
  n13 = AppendVector(A=Constant() {r=0.653}, B=Constant() {r=0.29})
  n14 = Constant() {r=0.037}
  n15 = TextureSample() {samplerType=SAMPLERTYPE_Masks, texture=/BRCosmetics/Accessories/FORT_Backpacks/Backpack_F_MED_WiryPerk/Textures/T_WiryPerk_Backpack_FX}
  NamedRerouteDeclaration(None=Multiply(A=SphereMask(A=@n12, B=@n13, Radius=@n14), B=@n15.R)) {name=ICMask, nodeColor=<rgba 0 1 0.4844 1>}
  n16 = Constant3Vector() {constant=<rgba 2 0 1.9952 1>}
  n17 = Constant3Vector() {constant=<rgba 0.35 0.274 0 1>}
  n18 = Constant3Vector() {constant=<rgba 0 0.3676 1.2 1>}
  n19 = ScalarParameter() {defaultValue=0.653, parameterName=IC_CenterU}
  n20 = ScalarParameter() {defaultValue=0.29, parameterName=IC_CenterV}
  n21 = Time()
  n22 = ScalarParameter() {defaultValue=0.2, parameterName=IC_rotationSpeed}
  n23 = TextureSample(UVs=MaterialFunctionCall(UVs=TextureCoordinate(), Rotation Center=AppendVector(A=@n19, B=@n20), "Rotation Angle (0-1)"=Multiply(A=@n21, B=@n22) {constB=0.1}) {materialFunction=&CustomRotator}.Rotated Values) {samplerType=SAMPLERTYPE_Masks, texture=&T_WiryPerk_Backpack_TA_mask}
  n24 = NamedRerouteUsage()
  n25 = Saturate(None=Multiply(A=Multiply(A=@n23.G, B=@n24), B=Constant() {r=2.2783}) {constB=3})
  NamedRerouteDeclaration(None=Multiply(
    A=MaterialFunctionCall(A=@n16, B=@n17, C=@n18, Alpha=@n25) {materialFunction=&Lerp_3Color}.Result
    B=NamedRerouteUsage()
   )) {name=IC_lightColor, nodeColor=<rgba 1 0.1641 0 1>}
  NamedRerouteDeclaration(None=@n11.B) {name=HueShiftMask, nodeColor=<rgba 1 0.1875 0 1>}
  n26 = Constant() {r=0.981}
  n27 = Constant() {r=0.852}
  NamedRerouteDeclaration(None=SphereMask(A=TextureCoordinate(), B=AppendVector(A=@n26, B=@n27), Radius=Constant() {r=0.01})) {name=HardDiskMask, nodeColor=<rgba 1 0 0.0703 1>}
  n28 = NamedRerouteUsage()
  n29 = TextureSample(UVs=NamedRerouteUsage()) {texture=/BRCosmetics/Accessories/FORT_Backpacks/Backpack_F_MED_WiryPerk/Textures/T_WiryPerk_Backpack_D}
  n30 = TextureSampleParameter2D() {
   group=00_Texture Inputs
   parameterName=Mask
   samplerSource=SSM_Wrap_WorldGroupSettings
   samplerType=SAMPLERTYPE_Masks
   sortPriority=3
   texture=&T_BlackMask
  }
  n31 = MaterialFunctionCall!(
   baseColor=NamedRerouteDeclaration(None=Multiply(A=@n28, B=@n29.RGB)) {name=BaseColor, nodeColor=<rgba 0 1 0.6484 1>}
   S_map
   input_fxmask
   WaterCooling Mask=NamedRerouteUsage()
   AO=ComponentMask(None=@n30.R) {g=true}
  ) {materialFunction=&/MF_WiryPerk_WaterCooling}
  NamedRerouteDeclaration(None=@n31.SlurpRougness) {name=waterCooling_rougness, nodeColor=<rgba 0.8516 1 0 1>}
  n32 = TextureSample() {samplerType=SAMPLERTYPE_Masks, texture=/BRCosmetics/Accessories/FORT_Backpacks/Backpack_F_MED_WiryPerk/Textures/T_Backpack_WiryPerkTA_Mask2}
  NamedRerouteDeclaration(None=@n32.G) {name=WaterCoolingG, nodeColor=<rgba 0 1 0.5781 1>}
  n33 = MaterialFunctionCall(
   UVs=TextureCoordinate()
   Rotation Center=AppendVector(A=Constant() {r=0.912}, B=Constant() {r=0.0862})
   "Rotation Angle (0-1)"=Multiply(A=Time(), B=ScalarParameter() {defaultValue=-0.3, group=WiryPerk_backpack, parameterName=Disc Rotator Speed}) {constB=-0.16}
  ) {materialFunction=&CustomRotator}
  NamedRerouteDeclaration(None=If(A=NamedRerouteUsage(), "A > B"=@n33.Rotated Values, "A == B"=@n33.Rotated Values, "A < B"=TextureCoordinate()) {constB=0.001}) {name=DiscRotator, nodeColor=<rgba 0 0.3437 1 1>}
  NamedRerouteDeclaration(None=Max(A=@n32.R, B=@n32.R)) {name=WaterCoolingRG, nodeColor=<rgba 1 0.7734 0 1>}
  NamedRerouteDeclaration(None=ScalarParameter() {group=WiryPerk_backpack, parameterName=TextAnimationA}) {name=Main_Animation, nodeColor=<rgba 0.5 0 1 1>}
  n34 = ScalarParameter() {defaultValue=1, group=WiryPerk_backpack, parameterName=TextAnimationB}
  n35 = NamedRerouteUsage()
  n36 = Multiply(A=Saturate(None=Subtract(A=@n35) {constB=2.5})) {constB=1.5}
  NamedRerouteDeclaration(None=Saturate(None=Subtract(A=@n34, B=@n36))) {name=End animation, nodeColor=<rgba 0 1 0.1797 1>}
  n37 = Multiply(
   A=Time()
   B=LinearInterpolate(B=ScalarParameter() {defaultValue=0.18, group=WiryPerk_backpack, parameterName=HueShiftSpeed}, Alpha=NamedRerouteUsage()) {constB=0.1}
  )
  n38 = TextureSample(UVs=NamedRerouteUsage()) {texture=/BRCosmetics/Accessories/FORT_Backpacks/Backpack_F_MED_WiryPerk/Textures/T_WiryPerk_Backpack_E}
  n39 = Multiply(A=NamedRerouteUsage(), B=@n38.RGB)
  n40 = Multiply(A=@n38.RGB) {constB=10}
  n41 = TextureSample(UVs=Panner(Coordinate=TextureCoordinate()) {speedY=1}) {
   constMipValue=0
   texture=/BRCosmetics/Characters/Player/Female/Medium/Bodies/F_MED_Twin_Dark/Skins/Twin_Light_Purple/Textures/T_twinlightPurpleBlurredNoise
  }
  n42 = Saturate(None=Multiply(A=Power(Base=@n41.R) {constExponent=4}) {constB=3})
  NamedRerouteDeclaration(None=LinearInterpolate(
    A=MaterialFunctionCall(Hue Shift Percentage=@n37, Texture=@n39) {materialFunction=&HueShift}.Result
    B=Multiply(A=@n40, B=@n42)
    Alpha=NamedRerouteUsage()
   )) {name=BaseEmissive, nodeColor=<rgba 0.4531 0 1 1>}
  NamedRerouteDeclaration(None=@n32.RGB) {name=WaterCoolingMask, nodeColor=<rgba 0 1 1 1>}
  NamedRerouteDeclaration(None=@n31.Slurp_emiss) {name=waterCooling_Emissive, nodeColor=<rgba 0.7109 1 0 1>}
  NamedRerouteDeclaration(None=@n31.Slurp_bascolor) {name=waterCooling_BaseColor, nodeColor=<rgba 0.9688 0 1 1>}
 ]}
