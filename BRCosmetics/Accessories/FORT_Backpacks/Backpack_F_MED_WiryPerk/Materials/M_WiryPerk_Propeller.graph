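# Additive, unlit, two-sided material for the WiryPerk back bling's propeller, drawing a spinning blurred fan with radial text from a GameOver texture.
# The fan blends two rotating blur textures, recolors them and adds panning radial specular detail.
# Radial-coordinate sphere masks driven by animation parameters reveal the wobbling text, and explicit UV derivatives stabilize radial sampling.
# Named reroutes carry these layers into color-lerped emissive and opacity calculations.
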Material /BRCosmetics/Accessories/FORT_Backpacks/Backpack_F_MED_WiryPerk/Materials/M_WiryPerk_Propeller
graph {
 properties {
  bUsedWithMeshDeformer=true
  bUsedWithSkeletalMesh=true
  blendMode=BLEND_Additive
  pixelDepthOffsetMode=PDOM_Legacy
  shadingModel=MSM_Unlit
  translucencyLightingMode=TLM_Surface
  twoSided=true
  usedShadingModels=Unlit
 }
 nodes [
  Constant()
  Constant() {r=0.1}
  MaterialFunctionCall(Emissive=LinearInterpolate(
    A=VectorParameter() {group=WiryPerk_backpack, parameterName=TextColor, defaultValue=<rgba 0.013 0.053 0.13 1>}.RGB
    B=VectorParameter() {group=WiryPerk_backpack, parameterName=FanColor, defaultValue=<rgba 0.8 6.3768 80 1>}.RGB
    Alpha=NamedRerouteUsage()
   )) {materialFunction=&MF_FX_EmissiveAdjustments}
  n1 = ScalarParameter() {defaultValue=3.5, group=WiryPerk_backpack, parameterName=SpinSpeed}
  Multiply(A=@n1) {constB=-1}
  n2 = Time()
  Multiply(A=@n2, B=DeltaTime())
  n3 = NamedRerouteUsage()
  n4 = ScalarParameter() {defaultValue=0.2, group=WiryPerk_backpack, parameterName=Text transparency overlay}
  n5 = Multiply(A=Add(A=NamedRerouteUsage(), B=ScalarParameter() {defaultValue=0.08, group=WiryPerk_backpack, parameterName=FanOpacity_offset}) {constB=0.05})
  n6 = Add(A=Add(A=Multiply(A=@n3, B=@n4), B=Multiply(A=@n3, B=@n5)), B=@n5)
  NamedRerouteDeclaration(None=Multiply(A=Saturate(None=@n6), B=Constant() {r=1})) {name=Opacity, nodeColor=<rgba 0.875 0 1 1>}
  n7 = MaterialFunctionCall(UVs=TextureCoordinate(), "Rotation Angle (0-1)"=Constant() {r=0.428}) {materialFunction=&CustomRotator}
  n8 = Multiply(A=MaterialFunctionCall(Vector or UVs=Subtract(A=@n7.Rotated Values) {constB=0.5}) {materialFunction=&VectorToRadialValue}.Radial Coordinates)
  n9 = ScalarParameter() {defaultValue=1, group=WiryPerk_backpack, parameterName=TextAnimationB}
  n10 = Saturate(None=Subtract(A=NamedRerouteUsage()) {constB=2.5})
  n11 = Subtract(A=@n9, B=@n10)
  n12 = ComponentMask(None=@n8) {r=true}
  n13 = Constant()
  n14 = NamedRerouteDeclaration(None=ScalarParameter() {group=WiryPerk_backpack, parameterName=TextAnimationA}) {name=Main_Animation, nodeColor=<rgba 0.5 0 1 1>}
  n15 = TextureCoordinate()
  n16 = ScalarParameter() {defaultValue=0.765, group=WiryPerk_backpack, parameterName=TextScale}
  n17 = Constant() {r=-1.125}
  n18 = Sine(None=Multiply(A=Time()) {constB=0.1})
  n19 = LinearInterpolate(A=Constant() {r=0.0005}, Alpha=Saturate(None=Multiply(A=@n18) {constB=2})) {constB=0.0007}
  n20 = OneMinus(None=ScalarParameter() {group=WiryPerk_backpack, parameterName=WiryPerkFrontEnd})
  n21 = Sine(None=Multiply(A=Time()) {constB=4})
  n22 = LinearInterpolate(A=@n17, B=Add(A=@n17, B=Multiply(A=@n19, B=@n20)), Alpha=Subtract(A=Multiply(A=@n21) {constB=2}))
  n23 = ScalarParameter() {group=WiryPerk_backpack, parameterName=TextAnimationA}
  n24 = Multiply(A=Saturate(None=Subtract(A=@n23) {constB=1.5})) {constB=-1}
  n25 = Multiply(A=MaterialFunctionCall(
    UVs=MaterialFunctionCall(UVs=@n15, Texture Scale=@n16) {materialFunction=&ScaleUVsByCenter}.UVs
    "Rotation Angle (0-1)"=Add(A=@n22, B=@n24)
   ) {materialFunction=&CustomRotator}.Rotated Values)
  n26 = TextureCoordinate()
  n27 = TextureSample!(
   UVs=MaterialFunctionCall(Vector or UVs=Add(A=@n25) {constB=-0.5}) {materialFunction=&VectorToRadialValue}.Radial Coordinates
   Tex
   "DDX(UVs)"=DDX(Value=@n26)
   "DDY(UVs)"=DDY(Value=@n26)
   Apply View MipBias
  ) {
   mipValueMode=TMVM_Derivative
   samplerSource=SSM_Clamp_WorldGroupSettings
   texture=/BRCosmetics/Accessories/FORT_Backpacks/Backpack_F_MED_WiryPerk/Textures/T_GameOver
  }
  n28 = Multiply(
   A=SphereMask(A=ComponentMask(None=@n8) {g=true}, B=Constant(), Radius=Saturate(None=@n11)) {hardnessPercent=95}
   B=Multiply(A=SphereMask(A=@n12, B=@n13, Radius=@n14) {hardnessPercent=95}, B=@n27.R)
  )
  NamedRerouteDeclaration(None=Saturate(None=Power(Base=@n28) {constExponent=1.5})) {name=Text, nodeColor=<rgba 0.4766 1 0 1>}
  n29 = TextureSample!(UVs=Rotator(Coordinate=TextureCoordinate(), Time=Multiply(A=@n2, B=@n1)) {speed=-25}, Tex, Apply View MipBias) {
   samplerSource=SSM_Clamp_WorldGroupSettings
   texture=/BRCosmetics/Accessories/FORT_Backpacks/Backpack_F_MED_WiryPerk/Textures/T_WiryPerk_Propeller_Blur
  }
  n30 = Multiply(A=@n29.RGB) {constB=1.5}
  n31 = TextureSample(UVs=Rotator(Coordinate=TextureCoordinate(), Time=Multiply(A=@n2, B=@n1)) {speed=-25}) {
   samplerSource=SSM_Clamp_WorldGroupSettings
   texture=/BRCosmetics/Accessories/FORT_Backpacks/Backpack_F_MED_WiryPerk/Textures/T_WiryPerk_Propeller_Blur_Fast
  }
  n32 = Multiply(A=@n31.RGB) {constB=1.5}
  n33 = ScalarParameter() {defaultValue=0.2, group=WiryPerk_backpack, parameterName=SpeedBlend}
  n34 = Desaturation(
   None=MaterialFunctionCall(Hue Shift Percentage=ScalarParameter() {parameterName=HueShift}, Texture=LinearInterpolate(A=@n30, B=@n32, Alpha=@n33) {constAlpha=1}) {materialFunction=&HueShift}.Result
   Fraction=ScalarParameter() {defaultValue=1, parameterName=Desaturate}
  )
  n35 = Constant() {r=6}
  n36 = Add(A=Multiply(A=TextureCoordinate(), B=@n35), B=Multiply(A=@n35) {constB=-0.5})
  n37 = ScalarParameter() {defaultValue=-3.5, group=WiryPerk_backpack, parameterName=Fan_detaail_Speed}
  n38 = Constant()
  n39 = TextureCoordinate()
  n40 = TextureSample!(
   UVs=Panner(
    Coordinate=MaterialFunctionCall(Vector or UVs=@n36) {materialFunction=&VectorToRadialValue}.Radial Coordinates
    Speed=AppendVector(A=@n37, B=@n38)
   ) {speedX=-3.5}
   Tex
   "DDX(UVs)"=DDX(Value=@n39)
   "DDY(UVs)"=DDY(Value=@n39)
   Apply View MipBias
  ) {mipValueMode=TMVM_Derivative, texture=/BRCosmetics/Accessories/FORT_Backpacks/Backpack_F_MED_WiryPerk/Textures/T_WiryPerk_Propeller_Spec}
  n41 = Multiply(A=Power(Base=@n40.R), B=Constant() {r=0.3}) {constB=0.05}
  n42 = ScalarParameter() {defaultValue=1, group=WiryPerk_backpack, parameterName=Fan_detaail}
  NamedRerouteDeclaration(None=Add(A=ComponentMask(None=@n34) {g=true}, B=Multiply(A=@n41, B=@n42))) {name=Fan, nodeColor=<rgba 0.3125 1 0 1>}
  n43 = Multiply(A=@n10) {constB=1.5}
  NamedRerouteDeclaration(None=Saturate(None=Subtract(A=@n9, B=@n43))) {name=End animation, nodeColor=<rgba 0 1 0.1797 1>}
  NamedRerouteUsage()
  n44 = NamedRerouteUsage()
  Saturate(None=Multiply(A=Multiply(A=@n44) {constB=2}, B=NamedRerouteUsage()))
  ScalarParameter() {defaultValue=0.6991, group=WiryPerk_backpack, parameterName=backpack_Anisotropy}
  ScalarParameter() {defaultValue=0.3, group=WiryPerk_backpack, parameterName=backpack_Roughness}
 ]
}
