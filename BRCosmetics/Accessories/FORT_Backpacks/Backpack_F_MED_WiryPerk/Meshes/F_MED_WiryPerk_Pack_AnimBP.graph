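# Backpack AnimBP for the WiryPerk pack that spins its fan bones procedurally.
# The AnimGraph adds bone-space rotations to fan_r, fan_l and sidefan from one rotator, and to disc from a second.
# Each update, Spin FNC advances the fan roll by speed times real-time delta, normalizes it, then stores the current time.
# The disc rotator has a matching spin chain, but its exec input shows no link to the update event.
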AnimBlueprint /BRCosmetics/Accessories/FORT_Backpacks/Backpack_F_MED_WiryPerk/Meshes/F_MED_WiryPerk_Pack_AnimBP
target_skeleton=&/F_MED_WiryPerk_Pack_Skeleton
graphs [
 AnimGraph {nodes [
   n1 = Get Spinning_Rotator() -> (Spinning_Rotator:Rotator)
   n2 = "Transform (Modify) Bone - Bone: fan_r"(ComponentPose=Local To Component(LocalPose=Input Pose()), Alpha=1, Translation="0,0,0", Rotation=@n1, Scale="1,1,1") {
    rotationMode=BMM_Additive
    rotationSpace=BCS_BoneSpace
    translationMode=BMM_Additive
    translationSpace=BCS_BoneSpace
    alphaBoolBlend {blendOption=Linear}
    boneToModify {boneName=fan_r}
   }
   n3 = "Transform (Modify) Bone - Bone: fan_l"(ComponentPose=@n2, Alpha=1, Translation="0,0,0", Rotation=@n1, Scale="1,1,1") {
    rotationMode=BMM_Additive
    rotationSpace=BCS_BoneSpace
    translationMode=BMM_Additive
    translationSpace=BCS_BoneSpace
    alphaBoolBlend {blendOption=Linear}
    boneToModify {boneName=fan_l}
   }
   n4 = "Transform (Modify) Bone - Bone: sidefan"(ComponentPose=@n3, Alpha=1, Translation="0,0,0", Rotation=@n1, Scale="1,1,1") {
    rotationMode=BMM_Additive
    rotationSpace=BCS_BoneSpace
    translationMode=BMM_Additive
    translationSpace=BCS_BoneSpace
    alphaBoolBlend {blendOption=Linear}
    boneToModify {boneName=sidefan}
   }
   n5 = "Transform (Modify) Bone - Bone: disc"(ComponentPose=@n4, Alpha=1, Translation="0,0,0", Rotation=Get Spinning_Rotator_00() -> (Spinning_Rotator_00:Rotator), Scale="1,1,1") {
    rotationMode=BMM_Additive
    rotationSpace=BCS_BoneSpace
    translationMode=BMM_Additive
    translationSpace=BCS_BoneSpace
    alphaBoolBlend {blendOption=Linear}
    boneToModify {boneName=disc}
   }
   Output Pose(Result=Component To Local(ComponentPose=@n5)) {layerGroup=None}
  ]}
 EventGraph {nodes [
   n1 = Spin FNC!(
    execute
    self
    speed:double=-150
    InitVal:double=Get!() -> (Spinning_Rotator_Roll:float, Spinning_Rotator_Pitch:float, Spinning_Rotator_Yaw:float)
   ) -> (then, output:double)
   > Set Spinning_Rotator_00(Spinning_Rotator_00:Rotator=MakeRotator(Roll=@n1)) -> (Output_Get:Rotator)
   Event BlueprintUpdateAnimation()
   > n2 = Sequence!(execute) -> (then_0, then_1, then_2:Exec)
   Reroute Node(InputPin=@n2.then_1)
   Set deltaT(execute=@n2.then_2, deltaT:double=GetRealTimeSeconds()) -> (Output_Get:double)
   n3 = Spin FNC(execute=@n2.then_0, speed=-150, InitVal=Get())
   > Set Spinning_Rotator(Spinning_Rotator:Rotator=MakeRotator(Roll=@n3)) -> (Output_Get:Rotator)
   TryGetPawnOwner()
  ]}
 Spin FNC {nodes [
   n1 = Spin FNC!() -> (then, speed:double=~, InitVal:double=~) {is=title K2Node_FunctionEntry}
   n2 = float * float(A=@n1, B=float - float(A=GetRealTimeSeconds(), B=Get deltaT() -> (deltaT:double)))
   > Return Node!(
    execute
    output:double=NormalizeAxis!(Angle:float=float + float(A=@n1.InitVal, B=@n2) {is=title K2Node_CommutativeAssociativeBinaryOperator}) -> (ReturnValue:float)
   )
  ]}
]
variables [
 deltaT {type=number}
 Spinning_Rotator {type=Rotator}
 Spinning_Rotator_00 {type=Rotator}
]
